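# Material function that computes a lightning bolt opacity profile for the CreamSkull Dove skydiving contrail effect.
# An offset UV component, defaulting to U, is remapped around the center, made absolute and divided by a width multiplier input.
# A core width parameter keeps the middle fully opaque and renormalizes the remaining distance.
# Saturation, square root, inversion and an exponent parameter shape a sharp falloff toward the edges.
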MaterialFunction /BRCosmetics/Athena/Effects/Particles/Skydiving/NS_Contrail_CreamSkull_Dove/Materials/MF_Contrail_CreamSkull_Dove_LightningOpacity
graph {nodes [
  n1 = ComponentMask(None=TextureCoordinate()) {r=true}
  n2 = ConstantBiasScale(None=Add(
    A=FunctionInput(Preview=@n1) {bUsePreviewValueAsDefault=true, inputName=Lightning Width UV Component, inputType=FunctionInput_Scalar}
    B=FunctionInput() {bUsePreviewValueAsDefault=true, inputName=Offset, inputType=FunctionInput_Scalar, sortPriority=1}
   )) {bias=-0.5, scale=2}
  n3 = ScalarParameter() {defaultValue=0.01, parameterName=Lightning Bolt Core Width}
  n4 = Subtract(
   A=Divide(
    A=Abs(None=@n2)
    B=FunctionInput() {bUsePreviewValueAsDefault=true, inputName=Width Multiplier, inputType=FunctionInput_Scalar, sortPriority=2}
   )
   B=@n3
  )
  n5 = OneMinus(None=@n3)
  n6 = SquareRoot(None=Saturate(None=Divide(A=@n4, B=@n5)))
  FunctionOutput(None=Power(Base=OneMinus(None=@n6), Exp=ScalarParameter() {defaultValue=8, parameterName=Lightning Bolt Falloff Exponent}))
 ]}
