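# Lit Niagara mesh-particle material for the clock face in the CreamSkull Dove skydiving contrail, animating its breakup into floating chunks.
# Chunk pivots in UV1 seed vector noise that randomizes time and speed for rotation, radial oscillation and expand/contract offsets, with normals rotated to match.
# Radius and UV masks separate face, rim and inner-rim shading.
# Emissive Intensity drives glow and a flare peaking at the mesh transition point, while WPO Enable delays breakup until then.
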Material /BRCosmetics/Athena/Effects/Particles/Skydiving/NS_Contrail_CreamSkull_Dove/Materials/M_Contrail_CreamSkull_Dove_Clock_Face
graph {
 properties {
  bAutomaticallySetUsageInEditor=false
  bUseMaterialAttributes=true
  bUsedWithNiagaraMeshParticles=true
  bUsedWithStaticMesh=false
  usedShadingModels=Default Lit
 }
 nodes [
  n1 = LinearInterpolate(
   A=VectorParameter() {parameterName=Base Color and Roughness Numbers, defaultValue=<rgba 0.015 0.015 0.015 0.85>}.RGBA
   B=LinearInterpolate(
    A=VectorParameter() {parameterName=Base Color and Roughness Rim, defaultValue=<rgba 0.02 0.0153 0.006 0.44>}.RGBA
    B=VectorParameter() {parameterName=Base Color and Roughness Rim Inner Faces, defaultValue=<rgba 0 0 0 1>}.RGBA
    Alpha=NamedRerouteUsage()
   )
   Alpha=NamedRerouteUsage()
  )
  n2 = TextureSample() {
   samplerType=SAMPLERTYPE_LinearGrayscale
   texture=/BRCosmetics/Athena/Effects/Particles/Skydiving/NS_Contrail_CreamSkull_Dove/Textures/T_Contrail_CreamSkull_Dove_Clock_R
  }
  n3 = Subtract(A=@n2.R) {constB=0.5}
  n4 = ScalarParameter() {defaultValue=0.4, parameterName=Roughness Map Influence}
  n5 = VectorParameter() {parameterName=Emissive Color, defaultValue=<rgba 1 1.6667 2 0>}
  n6 = Multiply(A=@n5.RGB, B=NamedRerouteUsage())
  n7 = Subtract(
   A=Length(None=LocalPosition() {includedOffsets=ExcludeOffsets}.XY) {is=class MaterialExpressionLength}
   B=ScalarParameter() {defaultValue=85, parameterName=Radius Threshold}
  )
  n8 = NamedRerouteDeclaration(None=Saturate(None=Ceil(None=@n7))) {name=Rim Mask, nodeColor=<rgba 0.5938 0 1 1>}
  n9 = NamedRerouteUsage()
  n10 = ComponentMask(None=NamedRerouteDeclaration(None=Transform(None=@n9) {transformType=TRANSFORM_Instance}) {name=Normal Mesh Particle Space, nodeColor=<rgba 1 0 0.0234 1>}) {g=true, r=true}
  n11 = Multiply(A=OneMinus(None=Length(None=@n10)), B=ScalarParameter() {defaultValue=0.6, parameterName=Emissive Contrast})
  n12 = ScalarParameter() {parameterName=Emissive Intensity}
  n13 = Multiply(A=Multiply(A=LinearInterpolate(A=@n5.RGB, B=@n6, Alpha=@n8), B=OneMinus(None=@n11)), B=@n12)
  n14 = NamedRerouteDeclaration(None=ScalarParameter() {defaultValue=0.2, parameterName=Clock Rim Mesh Transition Point}) {name=Clock Rim Mesh Transition Point, nodeColor=<rgba 0.0547 0 1 1>}
  n15 = MaterialFunctionCall!(Gradient=@n12, Center Point=@n14) {materialFunction=/Game/Effects/Fort_Effects/MaterialFunctions/MF_GradientReflected}
  n16 = Subtract(A=@n12, B=@n14)
  n17 = LinearInterpolate(
   A=ScalarParameter() {defaultValue=7, parameterName=Clocm Rim Mesh Transition Flare Up Exponent}
   B=ScalarParameter() {defaultValue=3, parameterName=Clocm Rim Mesh Transition Flare Up Exponent Broken}
   Alpha=Saturate(None=Sign(None=@n16))
  )
  n18 = VectorParameter() {parameterName=Emissive Flare Up Color, defaultValue=<rgba 0.4 1.2 2 2>}
  n19 = DotProduct(A=NamedRerouteUsage(), B=NamedRerouteUsage())
  n20 = Max(A=Power(Base=Abs(None=@n19), Exp=@n18.A), B=NamedRerouteUsage())
  n21 = Multiply(A=Multiply(A=Power(Base=@n15.Result, Exp=@n17), B=Multiply(A=@n18.RGB, B=@n20)), B=NamedRerouteUsage())
  MakeMaterialAttributes!(
   BaseColor=ComponentMask(None=@n1) {b=true, g=true, r=true}
   Metallic
   Specular
   Roughness=Add(A=ComponentMask(None=@n1) {a=true}, B=Multiply(A=@n3, B=@n4))
   Anisotropy
   EmissiveColor=MaterialFunctionCall(Emissive=Add(A=@n13, B=@n21)) {materialFunction=&MF_FX_EmissiveAdjustments}.Result
   Opacity
   OpacityMask
   Normal=NamedRerouteUsage()
   Tangent
   WorldPositionOffset=NamedRerouteUsage()
   SubsurfaceColor
   ClearCoat
   ClearCoatRoughness
   AmbientOcclusion
   Refraction
   CustomizedUVs_0
   CustomizedUVs_1
   CustomizedUVs_2
   CustomizedUVs_3
   CustomizedUVs_4
   CustomizedUVs_5
   CustomizedUVs_6
   CustomizedUVs_7
   PixelDepthOffset
   ShadingModel
   Displacement
  )
  n22 = Time()
  n23 = ScalarParameter() {defaultValue=0.33, parameterName=Time Scale}
  n24 = NamedRerouteUsage()
  NamedRerouteDeclaration(None=Add(A=Multiply(A=@n22, B=@n23), B=Multiply(A=@n24) {constB=236.789})) {name=Time Scaled and Offset, nodeColor=<rgba 1 0.3281 0 1>}
  n25 = ScalarParameter() {defaultValue=0.4, parameterName=Radial Oscillation Speed Min}
  n26 = ScalarParameter() {defaultValue=1.8, parameterName=Radial Oscillation Speed Max}
  n27 = NamedRerouteUsage()
  n28 = Sine(None=Multiply(A=NamedRerouteUsage(), B=LinearInterpolate(A=@n25, B=@n26, Alpha=@n27)))
  n29 = ScalarParameter() {defaultValue=1.5, parameterName=Radial Oscillation Range}
  NamedRerouteDeclaration(None=Multiply(A=Multiply(A=@n28, B=@n29), B=NamedRerouteUsage())) {name=Radial Oscillation WPO, nodeColor=<rgba 0 1 0.4609 1>}
  n30 = TextureSample() {
   samplerType=SAMPLERTYPE_Normal
   texture=/BRCosmetics/Athena/Effects/Particles/Skydiving/NS_Contrail_CreamSkull_Dove/Textures/T_Contrail_CreamSkull_Dove_Clock_N
  }
  n31 = Transform(None=@n30.RGB)
  n32 = NamedRerouteUsage()
  n33 = Multiply(
   A=NamedRerouteUsage()
   B=LinearInterpolate(
    A=ScalarParameter() {defaultValue=0.618, parameterName=Rotation Speed Min}
    B=ScalarParameter() {defaultValue=1.618, parameterName=Rotation Speed Max}
    Alpha=NamedRerouteUsage()
   )
  )
  n34 = Multiply(
   A=Multiply(A=Sine(None=@n33), B=ScalarParameter() {defaultValue=0.02, parameterName=Rotation Range})
   B=ScalarParameter() {parameterName=Rotation Scale}
  )
  n35 = Add(A=RotateAboutAxis(NormalizedRotationAxis=@n32, RotationAngle=@n34, PivotPoint=Constant(), Position=@n31) {desc=Rotating normals}, B=@n31)
  NamedRerouteDeclaration(None=Transform(None=Normalize(VectorInput=@n35) {desc="This is the rotated vector, in world space"}) {transformSourceType=TRANSFORMSOURCE_World, transformType=TRANSFORM_Tangent}) {name=Normal, nodeColor=<rgba 0.5234 1 0 1>}
  n36 = NamedRerouteUsage()
  n37 = WorldPosition()
  n38 = Add(
   A=Add(A=RotateAboutAxis(NormalizedRotationAxis=@n32, RotationAngle=@n34, PivotPoint=@n36, Position=@n37.XYZ), B=NamedRerouteUsage())
   B=NamedRerouteUsage()
  )
  n39 = NamedRerouteUsage()
  n40 = NamedRerouteUsage()
  n41 = Divide(A=Subtract(A=ScalarParameter() {parameterName=WPO Enable}, B=@n40), B=OneMinus(None=@n40))
  NamedRerouteDeclaration(None=Multiply(A=Add(A=@n38, B=@n39), B=Saturate(None=@n41) {desc=This will delay the breakup so it starts once the meshes transition})) {name=WPO, nodeColor=<rgba 1 0 0.8437 1>}
  n42 = LocalPosition()
  n43 = Normalize(VectorInput=@n42.XYZ)
  n44 = Saturate(None=DotProduct(A=Multiply(A=@n43) {constB=-1}, B=NamedRerouteUsage()))
  n45 = Divide(
   A=Subtract(A=Length(None=@n42.XY), B=ScalarParameter() {defaultValue=50, parameterName=Rim Emissive Falloff Offset})
   B=ScalarParameter() {defaultValue=66, parameterName=Rim Emissive Falloff Distance}
  )
  n46 = NamedRerouteDeclaration(None=OneMinus(None=Saturate(None=@n45))) {name=Rim Emissive Falloff, nodeColor=<rgba 0.7578 1 0 1>}
  NamedRerouteDeclaration(None=Power(Base=Multiply(A=@n44, B=@n46), Exp=ScalarParameter() {defaultValue=2, parameterName=Rim Emissive Exponent})) {name=Rim Emissive Mask, nodeColor=<rgba 0.1719 0 1 1>}
  n47 = Subtract(A=ComponentMask(None=TextureCoordinate()) {g=true}, B=ScalarParameter() {defaultValue=0.5, parameterName=Rim Inner Faces Mask Offset})
  NamedRerouteDeclaration(None=Saturate(None=Divide(A=@n47) {constB=0.01})) {name=Rim Inner Faces Mask, nodeColor=<rgba 0 0.4844 1 1>}
  n48 = Constant3Vector() {constant=<rgba 0 0 1 0>}
  n49 = Multiply(A=NamedRerouteUsage(), B=ScalarParameter() {defaultValue=1.6, parameterName=Axis Rotation Speed A})
  n50 = Constant()
  n51 = Constant3Vector() {constant=<rgba 1 0 0 0>}
  n52 = Normalize(VectorInput=Add(A=RotateAboutAxis(NormalizedRotationAxis=@n48, RotationAngle=@n49, PivotPoint=@n50, Position=@n51), B=@n51))
  n53 = RotateAboutAxis(
   NormalizedRotationAxis=Constant3Vector() {constant=<rgba 0 1 0 0>}
   RotationAngle=Multiply(A=NamedRerouteUsage(), B=ScalarParameter() {defaultValue=1.2, parameterName=Axis Rotation Speed B})
   PivotPoint=Constant()
   Position=@n52
  )
  NamedRerouteDeclaration(None=Normalize(VectorInput=Add(A=@n53, B=@n52))) {name=Rotation Axis, nodeColor=<rgba 1 0 0.5391 1>}
  n54 = AppendVector(A=TextureCoordinate() {coordinateIndex=1}, B=Constant())
  n55 = VectorNoise(World Position=@n54)
  NamedRerouteDeclaration(None=ComponentMask(None=@n55) {r=true}) {name=Random Time Offset, nodeColor=<rgba 1 0 0.8672 1>}
  NamedRerouteDeclaration(None=TransformPosition(None=@n54) {transformSourceType=TRANSFORMPOSSOURCE_Instance, transformType=TRANSFORMPOSSOURCE_World}) {name=Chunk Pivot, nodeColor=<rgba 0 1 0.25 1>}
  NamedRerouteDeclaration(None=ComponentMask(None=@n55) {g=true}) {name=Random Free Rotation Speed, nodeColor=<rgba 0 0.4844 1 1>}
  n56 = NamedRerouteDeclaration(None=ParticlePositionWS()) {name=Particle Position, nodeColor=<rgba 1 0.2578 0 1>}
  n57 = WorldPosition()
  n58 = NamedRerouteDeclaration(None=Normalize(VectorInput=Subtract(A=@n56, B=@n57.XYZ))) {name=Radial Vector Pixel, nodeColor=<rgba 1 0.3984 0 1>}
  n59 = ScalarParameter() {defaultValue=5, parameterName=Mesh Contract}
  n60 = NamedRerouteDeclaration(None=Normalize(VectorInput=Transform(None=@n54) {transformSourceType=TRANSFORMSOURCE_Instance})) {name=Radial Vector Chunk, nodeColor=<rgba 1 0.3047 0 1>}
  n61 = ScalarParameter() {defaultValue=8, parameterName=Mesh Expand}
  NamedRerouteDeclaration(None=Add(A=Multiply(A=@n58, B=@n59), B=Multiply(A=@n60, B=@n61))) {name=Mesh Expand and Contract, nodeColor=<rgba 1 0 0.4922 1>}
  NamedRerouteDeclaration(None=ComponentMask(None=@n55) {b=true}) {name=Random Vertical Rotation Speed, nodeColor=<rgba 0 0.8125 1 1>}
  n62 = Multiply(
   A=NamedRerouteUsage()
   B=LinearInterpolate(
    A=ScalarParameter() {defaultValue=4, parameterName=Vertical Rotation Speed Min}
    B=ScalarParameter() {defaultValue=2, parameterName=Vertical Rotation Speed Max}
    Alpha=NamedRerouteUsage()
   )
  )
  n63 = RotateAboutAxis(
   NormalizedRotationAxis=Constant3Vector() {constant=<rgba 0 0 1 0>}
   RotationAngle=Multiply(A=Sine(None=@n62), B=ScalarParameter() {defaultValue=0.0025, parameterName=Vertical Rotation Range})
   PivotPoint=Constant()
   Position=AppendVector(A=TextureCoordinate() {coordinateIndex=1}, B=Constant())
  ) {desc="Doing it all in local space and then transforming afterwards\r\nNot even rotating normals since the range here is so small it wouldn't make a difference"}
  NamedRerouteDeclaration(None=Transform(None=@n63) {transformSourceType=TRANSFORMSOURCE_Instance}) {name=Vertical Rotation WPO, nodeColor=<rgba 0 0.0156 1 1>}
 ]
}
