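# Additive, unlit, two-sided glow material for the clock element of the CreamSkull Dove skydiving contrail, enabled for Niagara mesh particles.
# Glow Color RGB runs through the shared FX emissive adjustment function.
# The glow texture's red channel feeds all three wired channel-selector inputs, then is raised to Glow Color's alpha as exponent.
# That mask is depth-faded and scaled by Glow Opacity.
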Material /BRCosmetics/Athena/Effects/Particles/Skydiving/NS_Contrail_CreamSkull_Dove/Materials/M_Contrail_CreamSkull_Dove_Clock_Glow
graph {
 properties {
  bAutomaticallySetUsageInEditor=false
  bUsedWithNiagaraMeshParticles=true
  bUsedWithStaticMesh=false
  blendMode=BLEND_Additive
  shadingModel=MSM_Unlit
  twoSided=true
  usedShadingModels=Unlit
 }
 nodes [
  n1 = VectorParameter() {parameterName=Glow Color, defaultValue=<rgba 0.12 0.16 0.2 1>}
  MaterialFunctionCall(Emissive=@n1.RGB) {materialFunction=&MF_FX_EmissiveAdjustments}
  n2 = TextureSample() {
   samplerType=SAMPLERTYPE_LinearColor
   texture=/BRCosmetics/Athena/Effects/Particles/Skydiving/NS_Contrail_CreamSkull_Dove/Textures/T_Contrail_CreamSkull_Dove_Clock_Glow
  }
  n3 = MaterialFunctionCall!(1=@n2.R, 2=@n2.R, 3=@n2.R, 4, A=ScalarParameter() {parameterName=Channel to Use}) {materialFunction=/Engine/Functions/Engine_MaterialFunctions02/Lerp_Multiple_Float}
  Multiply(
   A=DepthFade(Opacity=Power(Base=@n3.Lerp 3 Inputs, Exp=@n1.A), FadeDistance=ScalarParameter() {defaultValue=22, parameterName=Depth Fade Distance})
   B=ScalarParameter() {defaultValue=1, parameterName=Glow Opacity}
  )
 ]
}
