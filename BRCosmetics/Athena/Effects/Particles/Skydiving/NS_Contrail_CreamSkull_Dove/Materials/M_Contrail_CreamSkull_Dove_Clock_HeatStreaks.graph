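# Translucent, unlit, two-sided Niagara mesh-particle material for heat streaks on the clock of the CreamSkull Dove skydiving contrail.
# Two panned chromatic caustics samples use only the V coordinate to form streaks, then are blended, contrast-shaped and tinted.
# U-axis end fades, a view-facing edge fade and depth fade shape opacity.
# Emissive Intensity blends from an idle color with streak-masked opacity to tinted streaks over a background color.
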Material /BRCosmetics/Athena/Effects/Particles/Skydiving/NS_Contrail_CreamSkull_Dove/Materials/M_Contrail_CreamSkull_Dove_Clock_HeatStreaks
graph {
 properties {
  bAutomaticallySetUsageInEditor=false
  bUsedWithNiagaraMeshParticles=true
  bUsedWithStaticMesh=false
  blendMode=BLEND_Translucent
  shadingModel=MSM_Unlit
  twoSided=true
  usedShadingModels=Unlit
 }
 nodes [
  n1 = ComponentMask(None=TextureCoordinate()) {r=true}
  n2 = Divide(A=@n1, B=ScalarParameter() {defaultValue=0.2, parameterName=Streaks Start Fade Size})
  n3 = Divide(A=OneMinus(None=@n1), B=ScalarParameter() {defaultValue=0.8, parameterName=Streaks End Fade Size})
  n4 = Power(Base=Multiply(A=Saturate(None=@n2), B=Saturate(None=@n3)), Exp=ScalarParameter() {defaultValue=3, parameterName=Streaks Fade Exponent})
  n5 = Abs(None=DotProduct(A=VertexNormalWS(), B=CameraVectorWS()))
  n6 = ScalarParameter() {defaultValue=1.5, parameterName=Fresnel Exponent}
  n7 = DepthFade(Opacity=Multiply(A=@n4, B=Power(Base=@n5, Exp=@n6)), FadeDistance=ScalarParameter() {defaultValue=33, parameterName=Depth Fade Distance})
  n8 = Saturate(None=ScalarParameter() {parameterName=Emissive Intensity})
  LinearInterpolate(A=Multiply(A=@n7, B=NamedRerouteUsage()), B=@n7, Alpha=@n8)
  n9 = ComponentMask(None=TextureCoordinate()) {g=true}
  n10 = AppendVector(A=@n9, B=@n9)
  n11 = VectorParameter() {parameterName=UV Scales, defaultValue=<rgba 1 1 1 3>}
  n12 = ComponentMask(None=@n11.RGB) {g=true, r=true}
  n13 = Multiply(A=Time(), B=ScalarParameter() {defaultValue=1.5, parameterName=Time Scale})
  n14 = VectorParameter() {parameterName=Panner Speeds, defaultValue=<rgba -0.2 0.3 -0.5 0.1>}
  n15 = TextureSample(UVs=Panner(Coordinate=Multiply(A=@n10, B=@n12), Time=@n13, Speed=ComponentMask(None=@n14.RGB) {g=true, r=true}) {bFractionalPart=true}) {texture=/Game/Effects/Fort_Effects/Textures/Tile/T_ShieldCaustics_Tile_Chromatic}
  n16 = VectorParameter() {parameterName=Texture Weights and Exponent, defaultValue=<rgba 1 1 2 0>}
  n17 = AppendVector(A=@n11.B, B=@n11.B)
  n18 = TextureSample(UVs=Panner(Coordinate=Multiply(A=@n10, B=@n17), Time=@n13, Speed=AppendVector(A=@n14.B, B=@n14.B)) {bFractionalPart=true}) {texture=&T_ShieldCaustics_Tile_Chromatic}
  n19 = Divide(A=Add(A=Multiply(A=@n15.RGB, B=@n16.R), B=Multiply(A=@n18.RGB, B=@n16.G)), B=Add(A=@n16.R, B=@n16.R))
  n20 = Power(Base=OneMinus(None=Saturate(None=@n19)), Exp=@n16.B)
  n21 = OneMinus(None=@n4)
  n22 = OneMinus(None=LinearInterpolate(A=@n20, B=Multiply(A=@n20, B=@n21), Alpha=ScalarParameter() {defaultValue=0.1, parameterName=Streaks Gradient Mix}))
  n23 = Multiply(A=@n22, B=VectorParameter() {parameterName=Texture Tint, defaultValue=<rgba 0.4 0.6667 2 0>}.RGB)
  n24 = VectorParameter() {parameterName=Background Color, defaultValue=<rgba 0 0.0667 0.1 0>}
  MaterialFunctionCall(Emissive=LinearInterpolate(A=VectorParameter() {parameterName=Emissive Color Idle, defaultValue=<rgba 0.4 0.3587 0.3046 1>}.RGB, B=Add(A=@n23, B=@n24.RGB), Alpha=@n8)) {materialFunction=&MF_FX_EmissiveAdjustments}
  NamedRerouteDeclaration(None=Desaturation(None=@n22)) {name=Opacity Idle, nodeColor=<rgba 1 0.0469 0 1>}
 ]
}
