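# Unlit translucent Niagara ribbon material that draws comet trails for the CreamSkull Dove skydiving contrail.
# Particle color feeds emissive through the shared MF_FX_EmissiveAdjustments function.
# The trail mask is procedural from UVs: a sine fade along U multiplies a falloff across V, whose hard center narrows as U increases.
# Width and falloff exponent are parameters, and particle alpha scales the mask.
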Material /BRCosmetics/Athena/Effects/Particles/Skydiving/NS_Contrail_CreamSkull_Dove/Materials/M_Contrail_CreamSkull_Dove_CometTrails
graph {
 properties {
  bAutomaticallySetUsageInEditor=false
  bUsedWithNiagaraRibbons=true
  bUsedWithStaticMesh=false
  blendMode=BLEND_Translucent
  shadingModel=MSM_Unlit
  usedShadingModels=Unlit
 }
 nodes [
  n1 = ParticleColor()
  MaterialFunctionCall(Emissive=@n1.RGB) {materialFunction=&MF_FX_EmissiveAdjustments}
  n2 = TextureCoordinate()
  n3 = ComponentMask(None=@n2) {r=true}
  n4 = Sine(None=Divide(A=@n3))
  n5 = ConstantBiasScale(None=ComponentMask(None=@n2) {g=true}) {bias=-0.5, scale=2}
  n6 = Multiply(A=OneMinus(None=@n3), B=ScalarParameter() {defaultValue=0.1, parameterName=Comet Hard Trail Width})
  n7 = Divide(A=Subtract(A=Abs(None=@n5), B=@n6), B=OneMinus(None=@n6))
  n8 = OneMinus(None=SquareRoot(None=Saturate(None=@n7)))
  n9 = ScalarParameter() {defaultValue=2, parameterName=Comet Trail Falloff Exponent}
  Multiply(A=@n1.A, B=Multiply(A=@n4, B=Power(Base=@n8, Exp=@n9)))
  SquareRoot(None=@n4)
 ]
}
