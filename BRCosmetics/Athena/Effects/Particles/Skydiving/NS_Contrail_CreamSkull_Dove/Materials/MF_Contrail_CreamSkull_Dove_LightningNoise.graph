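# Material function that builds animated, four-layer 1D lightning noise for the CreamSkull Dove skydiving contrail materials.
# A gradient input, defaulting to UV V, is scaled by four frequencies and offset by time times four speeds.
# Each result samples one row of the noise texture's red channel as a separate octave.
# The octaves are remapped to a signed range, then combined as a weighted average normalized by total weight.
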MaterialFunction /BRCosmetics/Athena/Effects/Particles/Skydiving/NS_Contrail_CreamSkull_Dove/Materials/MF_Contrail_CreamSkull_Dove_LightningNoise
graph {nodes [
  n1 = ComponentMask(None=TextureCoordinate()) {g=true}
  n2 = VectorParameter() {parameterName=Noise Frequencies, defaultValue=<rgba 0.3 0.5 0.8 1.3>}
  n3 = Time()
  n4 = VectorParameter() {parameterName=Noise Speeds, defaultValue=<rgba 0.2 -0.3 0.4 -0.5>}
  n5 = Add(
   A=Multiply(
    A=FunctionInput(Preview=@n1) {bUsePreviewValueAsDefault=true, inputName=Gradient, inputType=FunctionInput_Scalar}
    B=FunctionInput(Preview=@n2.RGBA) {bUsePreviewValueAsDefault=true, inputName=Noise Frequencies, inputType=FunctionInput_Vector4, sortPriority=2}
   )
   B=Multiply(
    A=FunctionInput(Preview=@n3) {bUsePreviewValueAsDefault=true, inputName=Time, inputType=FunctionInput_Scalar, sortPriority=5}
    B=FunctionInput(Preview=@n4.RGBA) {bUsePreviewValueAsDefault=true, inputName=Noise Speeds, inputType=FunctionInput_Vector4, sortPriority=3}
   )
  )
  n6 = Constant()
  n7 = FunctionInput() {inputName=Noise Texture, inputType=FunctionInput_Texture2D, sortPriority=1}
  n8 = TextureSample(UVs=AppendVector(A=ComponentMask(None=@n5) {r=true}, B=@n6), Tex=@n7)
  n9 = TextureSample(UVs=AppendVector(A=ComponentMask(None=@n5) {g=true}, B=@n6), Tex=@n7)
  n10 = TextureSample(UVs=AppendVector(A=ComponentMask(None=@n5) {b=true}, B=@n6), Tex=@n7)
  n11 = TextureSample(UVs=AppendVector(A=ComponentMask(None=@n5) {a=true}, B=@n6), Tex=@n7)
  n12 = FunctionInput(Preview=VectorParameter() {parameterName=Noise Weights, defaultValue=<rgba 8 5 3 1>}.RGBA) {bUsePreviewValueAsDefault=true, inputName=Noise Weights, inputType=FunctionInput_Vector4, sortPriority=4}
  n13 = Multiply(
   A=ConstantBiasScale(None=MaterialFunctionCall(R=@n8.R, G=@n9.R, B=@n10.R, A=@n11.R) {materialFunction=&AppendMany}.RGBA) {bias=-0.5, scale=2}
   B=@n12
  )
  FunctionOutput(None=Divide(
    A=MaterialFunctionCall(f4=@n13) {materialFunction=&AddComponents}.f4
    B=MaterialFunctionCall(f4=@n12) {materialFunction=&AddComponents}.f4
   ))
 ]}
