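# Translucent unlit Niagara sprite material that draws an animated lightning bolt for the CreamSkull Dove skydiving contrail.
# A per-particle randomly shortened UV gradient feeds an erosion reveal, which switches to a reflected-gradient fade over particle lifetime.
# Panned bolt and blurred glow textures share UVs offset sideways by time-animated blue noise, randomized per particle.
# The combined mask is softened by depth fade and view-angle falloff, with a tinted HDR color driving emissive.
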Material /BRCosmetics/Athena/Effects/Particles/Skydiving/NS_Contrail_CreamSkull_Dove/Materials/M_Contrail_CreamSkull_Dove_Lightning_Bolt
graph {
 properties {
  bAutomaticallySetUsageInEditor=false
  bUsedWithNiagaraSprites=true
  bUsedWithStaticMesh=false
  blendMode=BLEND_Translucent
  shadingModel=MSM_Unlit
  usedShadingModels=Unlit
 }
 nodes [
  MaterialFunctionCall(Emissive=VectorParameter() {parameterName=Color, defaultValue=<rgba 0.42 2.28 6 33>}.RGB) {materialFunction=&MF_FX_EmissiveAdjustments}
  n1 = TextureCoordinate()
  n2 = OneMinus(None=Power(Base=ParticleRandom(), Exp=ScalarParameter() {defaultValue=2.5, parameterName=UV Shortened Random Exponent}))
  n3 = OneMinus(None=Saturate(None=ScalarParameter() {defaultValue=0.15, parameterName=UV Shortened Min Size}))
  n4 = Divide(A=OneMinus(None=ComponentMask(None=@n1) {g=true}), B=OneMinus(None=Multiply(A=@n2, B=@n3)))
  n5 = Frac(None=NamedRerouteDeclaration(None=Saturate(None=@n4)) {name=UV Shortened, nodeColor=<rgba 0.0781 0 1 1>})
  n6 = ParticleRelativeTime()
  n7 = ScalarParameter() {defaultValue=0.5, parameterName=Reveal Duration}
  n8 = Saturate(None=Divide(A=@n6, B=@n7))
  n9 = ScalarParameter() {defaultValue=7, parameterName=Reveal Exponent}
  n10 = MaterialFunctionCall!(
   Gradient=@n5
   Erosion Amount=NamedRerouteDeclaration(None=Power(Base=@n8, Exp=@n9)) {name=Reveal Timeline, nodeColor=<rgba 0.5234 0 1 1>}
   Softness=ScalarParameter() {defaultValue=0.2, parameterName=Reveal Softness}
  ) {materialFunction=/BRCosmetics/Athena/Effects/Particles/Emotes/NS_Emote_Spectrum/Functions/MF_Erosion}
  n11 = Subtract(A=@n6, B=@n7)
  n12 = Divide(A=@n11, B=OneMinus(None=@n7))
  n13 = MaterialFunctionCall!(
   Gradient=NamedRerouteDeclaration(None=Saturate(None=@n12)) {name=Fade Timeline, nodeColor=<rgba 0 1 0.2031 1>}
   Center Point=ScalarParameter() {defaultValue=0.2, parameterName=Fade Reflected Gradient Center}
  ) {materialFunction=/Game/Effects/Fort_Effects/MaterialFunctions/MF_GradientReflected}
  n14 = ScalarParameter() {defaultValue=2, parameterName=Fade Reflected Gradient Exponent}
  n15 = LinearInterpolate(
   A=NamedRerouteDeclaration(None=OneMinus(None=@n10.Increased Range)) {name=Reveal Opacity, nodeColor=<rgba 0 0.7891 1 1>}
   B=Multiply(A=Power(Base=@n13.Result, Exp=@n14), B=ScalarParameter() {defaultValue=2, parameterName=Fade Reflected Gradient Boost})
   Alpha=Saturate(None=Sign(None=@n11))
  )
  n16 = TextureCoordinate()
  n17 = Multiply(A=ComponentMask(None=@n16) {g=true}, B=ScalarParameter() {defaultValue=1.62, parameterName=Lightning Texture Tile A})
  n18 = ParticleRandom()
  n19 = Add(A=Saturate(None=ComponentMask(None=@n16) {r=true}), B=Multiply(A=NamedRerouteUsage(), B=NamedRerouteUsage()))
  n20 = Panner(
   Coordinate=AppendVector(A=Add(A=@n17, B=@n18), B=Saturate(None=@n19))
   Speed=AppendVector(A=ScalarParameter() {defaultValue=1.3, parameterName=Lighting Texture Panner Speed A}, B=Constant())
  ) {bFractionalPart=true}
  n21 = TextureSample(UVs=@n20) {
   samplerType=SAMPLERTYPE_LinearGrayscale
   texture=/BRCosmetics/Athena/Effects/Particles/Skydiving/NS_Contrail_CreamSkull_Dove/Textures/T_Contrail_CreamSkull_Dove_Lightning
  }
  n22 = Power(Base=@n21.R, Exp=ScalarParameter() {defaultValue=2, parameterName=Lightning Texture Exponent A})
  n23 = TextureSample(UVs=@n20) {
   samplerType=SAMPLERTYPE_LinearGrayscale
   texture=/BRCosmetics/Athena/Effects/Particles/Skydiving/NS_Contrail_CreamSkull_Dove/Textures/T_Contrail_CreamSkull_Dove_Lightning_Blur
  }
  n24 = Multiply(
   A=Power(Base=@n23.R, Exp=ScalarParameter() {defaultValue=2, parameterName=Lightning Glow Texture Exponent A})
   B=ScalarParameter() {defaultValue=0.1, parameterName=Lighting Glow Texture A Multiplier}
  )
  n25 = Power(Base=Add(A=Add(A=@n22, B=@n24)) {constB=0}, Exp=Divide(B=NamedRerouteUsage()))
  n26 = DotProduct(A=VertexNormalWS(), B=CameraVectorWS())
  Multiply(
   A=DepthFade(Opacity=Multiply(A=@n15, B=@n25), FadeDistance=ScalarParameter() {defaultValue=22, parameterName=Depth Fade Distance})
   B=MaterialFunctionCall!(In=Abs(None=@n26), Power=ScalarParameter() {defaultValue=3, parameterName=Fresnel Exponent}) {materialFunction=/Game/Effects/Fort_Effects/MaterialFunctions/MF_SCurve_Float}.Result
  )
  n27 = Time()
  n28 = ScalarParameter() {defaultValue=0.15, parameterName=Lightning Noises Time Scale}
  n29 = ParticleRandom()
  n30 = MaterialFunctionCall!(
   Gradient=ComponentMask(None=TextureCoordinate()) {g=true}
   Noise Texture=TextureObject() {samplerType=SAMPLERTYPE_LinearGrayscale, texture=/BRCosmetics/Global/Textures/Default/Noise/T_BlueNoise}
   Noise Frequencies=Multiply(
    A=VectorParameter() {parameterName=Lightning Noise A Frequencies, defaultValue=<rgba 0.618 1 1.618 2.618>}.RGBA
    B=ScalarParameter() {defaultValue=0.05, parameterName=Lightning Noise A Frequency Multiplier}
   )
   Noise Speeds=VectorParameter() {parameterName=Lightning Noise A Speeds, defaultValue=<rgba 0.62 -1 1.62 -2.62>}.RGBA
   Noise Weights=VectorParameter() {parameterName=Lightning Noise A Weights, defaultValue=<rgba 2.62 1.62 1 0.62>}.RGBA
   Time=Add(A=Multiply(A=@n27, B=@n28), B=Multiply(A=@n29) {constB=12.456})
  ) {materialFunction=&/MF_Contrail_CreamSkull_Dove_LightningNoise}
  NamedRerouteDeclaration(None=@n30.Result) {name=Lightning Noise A, nodeColor=<rgba 0.4297 0 1 1>}
  NamedRerouteDeclaration(None=MaterialFunctionCall(Gradient=Multiply(A=@n5), Center Point=ScalarParameter() {defaultValue=0.2, parameterName=Reflected Gradient Global Opacity}) {materialFunction=&MF_GradientReflected}.Result) {name=Ends Exponent Fade, nodeColor=<rgba 0 1 0.6953 1>}
  NamedRerouteDeclaration(None=Multiply(A=NamedRerouteUsage(), B=ScalarParameter() {defaultValue=0.75, parameterName=Lightning Noise Offset Intensity})) {name=Lightning Sideways Offset A, nodeColor=<rgba 0 0.8594 1 1>}
 ]
}
