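# Material function for the CreamSkull Dove skydiving contrail that computes crawling lightning and a contact-point opacity mask.
# It pans V over time, doubles and wraps it, then samples a data texture at a set mip for lightning gradient (R) and size (B).
# Contact size over lightning size remaps the gradient into coordinates for a radial core falloff, faded where the area is too small.
# Outputs are contact opacity, gradient, both sizes and fade.
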MaterialFunction /BRCosmetics/Athena/Effects/Particles/Skydiving/NS_Contrail_CreamSkull_Dove/Materials/MF_Contrail_CreamSkull_Dove_Lightning_Crawling
graph {nodes [
  n1 = FunctionInput(Preview=TextureCoordinate()) {bUsePreviewValueAsDefault=true, inputName=UVs, inputType=FunctionInput_Vector2}
  n2 = Panner(
   Coordinate=AppendVector(A=ComponentMask(None=@n1) {g=true}, B=Constant())
   Speed=AppendVector(A=Constant(), B=ScalarParameter() {defaultValue=0.22, parameterName=Test Texture Pan Speed})
  ) {bFractionalPart=true}
  n3 = Multiply(A=Frac(None=ComponentMask(None=@n2) {g=true})) {constB=2}
  n4 = TextureSample!(
   UVs=AppendVector(A=ComponentMask(None=@n2) {r=true}, B=Frac(None=@n3))
   Tex=FunctionInput() {inputName=Texture, inputType=FunctionInput_Texture2D}
   Level=ScalarParameter() {parameterName=Mip level}
   Apply View MipBias
  ) {
   mipValueMode=TMVM_MipLevel
   samplerType=SAMPLERTYPE_LinearColor
   texture=/BRCosmetics/Athena/Effects/Particles/Skydiving/NS_Contrail_CreamSkull_Dove/Textures/T_Contrail_CreamSkull_Dove_LightningRing_Data
  }
  n5 = Saturate(None=Floor(None=@n3))
  n6 = LinearInterpolate(A=@n4.R, B=@n4.R, Alpha=@n5)
  n7 = LinearInterpolate(A=@n4.B, B=@n4.B, Alpha=@n5)
  n8 = Divide(
   A=NamedRerouteDeclaration(None=ScalarParameter() {defaultValue=0.075, parameterName=Contact Point Size}) {name=Contact Point Size, nodeColor=<rgba 1 0.8203 0 1>}
   B=@n7
  )
  n9 = Divide(A=@n6, B=@n8)
  n10 = Divide(A=OneMinus(None=@n6), B=@n8)
  n11 = AppendVector(A=ComponentMask(None=@n1) {r=true}, B=Multiply(A=Saturate(None=@n9), B=Saturate(None=@n10)))
  n12 = Constant() {r=0.5}
  n13 = Saturate(None=Divide(A=Distance(A=@n11, B=@n12), B=@n12))
  n14 = ScalarParameter() {defaultValue=0.03, parameterName=Contact Point Core Size}
  n15 = Divide(A=Subtract(A=@n13, B=@n14), B=OneMinus(None=@n14))
  n16 = OneMinus(None=SquareRoot(None=Saturate(None=@n15)))
  n17 = SquareRoot(None=Saturate(None=Subtract(A=@n8)))
  n18 = Power(Base=NamedRerouteDeclaration(None=OneMinus(None=@n17) {desc=To fade out contact point when area is too small}) {name=Contact Point Fade, nodeColor=<rgba 1 0 0.8906 1>}) {constExponent=4}
  FunctionOutput(None=Power(Base=Multiply(A=@n16, B=@n18), Exp=ScalarParameter() {defaultValue=1, parameterName=Contact Point Exponent})) {outputName=Contact Point Opacity}
  FunctionOutput(None=NamedRerouteUsage()) {outputName=Lightning Gradient}
  FunctionOutput(None=NamedRerouteUsage()) {outputName=Lightning Size}
  FunctionOutput(None=NamedRerouteUsage()) {outputName=Contact Point Size}
  FunctionOutput(None=NamedRerouteUsage()) {outputName=Contact Point Fade}
  NamedRerouteDeclaration(None=@n6) {name=Lightning Gradient, nodeColor=<rgba 1 0 0.5391 1>}
  NamedRerouteDeclaration(None=@n7) {name=Lightning Size, nodeColor=<rgba 0 0.8594 1 1>}
  n19 = OneMinus(None=Abs(None=ConstantBiasScale(None=@n11) {bias=-0.5, scale=2}))
  OneMinus(None=Multiply(A=ComponentMask(None=@n19) {r=true}, B=ComponentMask(None=@n19) {g=true}))
  Power(Base=OneMinus(None=@n13)) {constExponent=4}
 ]}
