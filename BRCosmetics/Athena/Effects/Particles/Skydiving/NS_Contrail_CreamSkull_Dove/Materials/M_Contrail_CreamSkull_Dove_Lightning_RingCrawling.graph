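# Translucent unlit Niagara mesh-particle material for crawling lightning rings in the CreamSkull Dove skydiving contrail.
# A crawling function reads a data texture to drive bolt gradient, size and contact-point glow, which brightens near scene depth.
# Two panned layers of bolt and blur textures are displaced by animated blue noise, with randomized, eroding idle bolts as an alternative.
# World position offset pushes the mesh with noise, widens it toward the camera as a ribbon and offsets it cameraward.
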Material /BRCosmetics/Athena/Effects/Particles/Skydiving/NS_Contrail_CreamSkull_Dove/Materials/M_Contrail_CreamSkull_Dove_Lightning_RingCrawling
graph {
 properties {
  bAutomaticallySetUsageInEditor=false
  bHasPixelAnimation=true
  bUsedWithNiagaraMeshParticles=true
  blendMode=BLEND_Translucent
  shadingModel=MSM_Unlit
  usedShadingModels=Unlit
 }
 nodes [
  Multiply(
   A=MaterialFunctionCall(Emissive=VectorParameter() {parameterName=Color, defaultValue=<rgba 0.42 2.28 6 0>}.RGB) {materialFunction=&MF_FX_EmissiveAdjustments}.Result
   B=NamedRerouteUsage()
  )
  NamedRerouteDeclaration(None=Constant() {r=1}) {name=Reveal, nodeColor=<rgba 0.1719 0 1 1>}
  n1 = WorldPosition()
  n2 = Subtract(A=@n1.XYZ, B=ParticlePositionWS())
  n3 = LinearInterpolate(
   A=NamedRerouteDeclaration(None=Normalize(VectorInput=@n2)) {name=Radial Vector, nodeColor=<rgba 0 0.6484 1 1>}
   B=Transform(None=Constant3Vector() {constant=<rgba 0 0 1 0>}) {transformSourceType=TRANSFORMSOURCE_Instance}
   Alpha=ScalarParameter() {defaultValue=0.5, parameterName=Outwards Upwards Blend}
  )
  n4 = Multiply(
   A=ScalarParameter() {defaultValue=44, parameterName=Crawl Noise Offset Intensity}
   B=NamedRerouteDeclaration(None=ScalarParameter() {defaultValue=1, parameterName=MeshScale}) {name=Mesh Scale, nodeColor=<rgba 0 1 0.1797 1>}
  )
  n5 = NamedRerouteUsage()
  n6 = LinearInterpolate(A=Multiply(A=NamedRerouteUsage(), B=Multiply(A=@n4, B=@n5)), Alpha=NamedRerouteUsage()) {constB=0}
  n7 = NamedRerouteUsage()
  n8 = ScalarParameter() {defaultValue=-1, parameterName=Radial Push}
  NamedRerouteDeclaration(None=Add(A=Multiply(A=@n3, B=@n6), B=Multiply(A=@n7, B=@n8))) {name=Lightning Outwards Push, nodeColor=<rgba 1 0 0.3516 1>}
  n9 = Multiply(A=NamedRerouteUsage(), B=NamedRerouteUsage())
  n10 = TextureObject() {samplerType=SAMPLERTYPE_LinearGrayscale, texture=/BRCosmetics/Global/Textures/Default/Noise/T_BlueNoise}
  n11 = Multiply(
   A=VectorParameter() {parameterName=Lightning Noise A Frequencies, defaultValue=<rgba 0.618 1 1.618 2.618>}.RGBA
   B=ScalarParameter() {defaultValue=0.2, parameterName=Lightning Noise A Frequency Multiplier}
  )
  n12 = VectorParameter() {parameterName=Lightning Noise A Speeds, defaultValue=<rgba 0.62 -1 1.62 -2.62>}
  n13 = VectorParameter() {parameterName=Lightning Noise A Weights, defaultValue=<rgba 2.62 1.62 1 0.62>}
  n14 = Multiply(A=Time(), B=ScalarParameter() {defaultValue=0.333, parameterName=Lightning Noises Time Scale})
  n15 = MaterialFunctionCall!(Gradient=@n9, Noise Texture=@n10, Noise Frequencies=@n11, Noise Speeds=@n12.RGBA, Noise Weights=@n13.RGBA, Time=@n14) {materialFunction=&/MF_Contrail_CreamSkull_Dove_LightningNoise}
  NamedRerouteDeclaration(None=VertexInterpolator(VS=@n15.Result).PS) {name=Lightning Noise A, nodeColor=<rgba 0.4297 0 1 1>}
  n16 = TextureObject() {
   samplerType=SAMPLERTYPE_LinearColor
   texture=/BRCosmetics/Athena/Effects/Particles/Skydiving/NS_Contrail_CreamSkull_Dove/Textures/T_Contrail_CreamSkull_Dove_LightningRing_Data
  }
  n17 = MaterialFunctionCall!(UVs=NamedRerouteUsage(), Texture=@n16) {materialFunction=&/MF_Contrail_CreamSkull_Dove_Lightning_Crawling}
  NamedRerouteDeclaration(None=LinearInterpolate(A=@n17.Lightning Size, Alpha=NamedRerouteUsage())) {name=Lightning Bolt Gradient Size, nodeColor=<rgba 1 0 0.7969 1>}
  n18 = ComponentMask(None=TextureCoordinate()) {g=true}
  n19 = Multiply(A=Time(), B=ScalarParameter() {defaultValue=0.5, parameterName=Idle Lightning Speed Scale})
  n20 = VectorNoise(World Position=Floor(None=@n19))
  n21 = ComponentMask(None=@n20) {desc=Random Position Offset, g=true}
  n22 = Subtract(A=@n21) {constB=0.5}
  n23 = LinearInterpolate(A=@n18, B=OneMinus(None=@n18), Alpha=Saturate(None=Sign(None=@n22)) {desc=Random Direction})
  n24 = LinearInterpolate(
   A=ScalarParameter() {defaultValue=0.2, parameterName=Idle Lighhtning Bolt Gradient Size Min}
   B=ScalarParameter() {defaultValue=0.5, parameterName=Idle Lighhtning Bolt Gradient Size Max}
   Alpha=ComponentMask(None=@n20) {desc=Random Size, r=true}
  )
  n25 = Divide(A=Frac(None=Subtract(A=@n23, B=@n21)), B=@n24)
  n26 = Floor(None=@n25)
  NamedRerouteDeclaration(None=OneMinus(None=Saturate(None=@n26))) {name=Idle Lightning Spline Width Multiplier, nodeColor=<rgba 0.2891 1 0 1>}
  n27 = NamedRerouteDeclaration(None=@n17.Lightning Gradient) {name=Lightning Bolt Gradient, nodeColor=<rgba 0.5937 1 0 1>}
  n28 = Divide(A=NamedRerouteUsage(), B=NamedRerouteUsage())
  n29 = Divide(A=@n27, B=@n28)
  n30 = Divide(A=OneMinus(None=@n27), B=@n28)
  n31 = Subtract(A=Multiply(A=Saturate(None=@n29), B=Saturate(None=@n30))) {constB=0.5}
  n32 = ScalarParameter() {defaultValue=0.5, parameterName=Lightning Bolt Contact Points Fade Size}
  n33 = NamedRerouteDeclaration(None=Saturate(None=Divide(A=@n31, B=@n32) {constB=0.5})) {name=Lightning Bolt Contact Points Fade, nodeColor=<rgba 0 1 0.2734 1>}
  n34 = NamedRerouteUsage()
  n35 = ScalarParameter() {defaultValue=0.5, parameterName=Lightning Noise Offset Intensity}
  NamedRerouteDeclaration(None=Multiply(A=@n33, B=Multiply(A=@n34, B=@n35))) {name=Lightning Sideways Offset A, nodeColor=<rgba 0.9453 1 0 1>}
  n36 = NamedRerouteUsage()
  NamedRerouteDeclaration(None=Multiply(A=@n33, B=Multiply(A=@n36, B=@n35))) {name=Lightning Sideways Offset B, nodeColor=<rgba 0.9453 1 0 1>}
  n37 = NamedRerouteUsage()
  n38 = ComponentMask(None=@n37) {g=true}
  n39 = ScalarParameter() {defaultValue=6, parameterName=Lightning Texture Tile A}
  n40 = Time()
  n41 = ScalarParameter() {defaultValue=1.75, parameterName=Lighting Texture Panner Speed A}
  n42 = ComponentMask(None=@n37) {r=true}
  n43 = Add(A=Divide(A=Subtract(A=@n42) {constB=0.5}, B=NamedRerouteUsage())) {constB=0.5}
  n44 = NamedRerouteUsage()
  n45 = ScalarParameter() {defaultValue=0.75, parameterName=Lightning Sideways Offset Multiplier A}
  n46 = AppendVector(A=Add(A=Multiply(A=@n38, B=@n39), B=Multiply(A=@n40, B=@n41)), B=Add(A=Saturate(None=@n43), B=Multiply(A=@n44, B=@n45)))
  n47 = TextureSample(UVs=@n46) {
   samplerType=SAMPLERTYPE_LinearGrayscale
   texture=/BRCosmetics/Athena/Effects/Particles/Skydiving/NS_Contrail_CreamSkull_Dove/Textures/T_Contrail_CreamSkull_Dove_Lightning
  }
  n48 = Power(Base=@n47.R, Exp=ScalarParameter() {defaultValue=2, parameterName=Lightning Texture Exponent A})
  n49 = TextureSample(UVs=@n46) {
   samplerType=SAMPLERTYPE_LinearGrayscale
   texture=/BRCosmetics/Athena/Effects/Particles/Skydiving/NS_Contrail_CreamSkull_Dove/Textures/T_Contrail_CreamSkull_Dove_Lightning_Blur
  }
  n50 = Multiply(
   A=Power(Base=@n49.R, Exp=ScalarParameter() {defaultValue=1, parameterName=Lightning Glow Texture Exponent A})
   B=ScalarParameter() {defaultValue=0.25, parameterName=Lighting Glow Texture A Multiplier}
  )
  n51 = ComponentMask(None=@n37) {g=true}
  n52 = ScalarParameter() {defaultValue=6, parameterName=Lightning Texture Tile B}
  n53 = Time()
  n54 = ScalarParameter() {defaultValue=-2, parameterName=Lighting Texture Panner Speed B}
  n55 = ComponentMask(None=@n37) {r=true}
  n56 = Add(A=Divide(A=Subtract(A=@n55) {constB=0.5}, B=NamedRerouteUsage())) {constB=0.5}
  n57 = NamedRerouteUsage()
  n58 = ScalarParameter() {defaultValue=1, parameterName=Lightning Sideways Offset Multiplier B}
  n59 = AppendVector(A=Add(A=Multiply(A=@n51, B=@n52), B=Multiply(A=@n53, B=@n54)), B=Add(A=Saturate(None=@n56), B=Multiply(A=@n57, B=@n58)))
  n60 = TextureSample(UVs=@n59) {
   samplerType=SAMPLERTYPE_LinearGrayscale
   texture=&T_Contrail_CreamSkull_Dove_Lightning
  }
  n61 = Multiply(
   A=Power(Base=@n60.R, Exp=ScalarParameter() {defaultValue=1, parameterName=Lightning Texture Exponent B})
   B=ScalarParameter() {defaultValue=0.3, parameterName=Lightning Texture B Multiplier}
  ) {constB=0.25}
  n62 = TextureSample(UVs=@n59) {
   samplerType=SAMPLERTYPE_LinearGrayscale
   texture=&T_Contrail_CreamSkull_Dove_Lightning_Blur
  }
  n63 = Multiply(A=@n62.R, B=ScalarParameter() {defaultValue=0.15, parameterName=Lighting Glow Texture B Multiplier})
  n64 = Multiply(A=Ceil(None=NamedRerouteUsage()), B=Add(A=Add(A=@n48, B=@n50), B=Add(A=@n61, B=@n63)))
  n65 = NamedRerouteUsage()
  n66 = ScalarParameter() {defaultValue=5, parameterName=Contact Point Floating Exponent}
  n67 = ConstantBiasScale(None=NamedRerouteUsage()) {bias=-0.5, scale=2}
  n68 = DotProduct(A=@n67, B=@n67)
  n69 = ConstantBiasScale(None=ComponentMask(None=NamedRerouteUsage()) {r=true}) {bias=-0.5, scale=2}
  n70 = DotProduct(A=@n69, B=@n69)
  n71 = Multiply(A=OneMinus(None=SquareRoot(None=@n68)), B=OneMinus(None=SquareRoot(None=@n70)))
  n72 = ScalarParameter() {defaultValue=0.333, parameterName=Lightning Glow Textures Combined Depth Light Contribution}
  n73 = Divide(A=Subtract(A=SceneDepth().R, B=PixelDepth().R), B=ScalarParameter() {defaultValue=33, parameterName=Contact Point Depth Range})
  n74 = Multiply(A=Add(A=@n65, B=Multiply(A=@n71, B=@n72)), B=OneMinus(None=Saturate(None=@n73)))
  n75 = ScalarParameter() {defaultValue=1, parameterName=Contact Point Depth Exponent}
  n76 = Add(
   A=Multiply(A=Power(Base=@n65, Exp=@n66), B=ScalarParameter() {defaultValue=0.6, parameterName=Contact Point Floating Brightness})
   B=Multiply(A=Power(Base=@n74, Exp=@n75), B=ScalarParameter() {defaultValue=0.4, parameterName=Contact Point Depth Brightness})
  )
  n77 = NamedRerouteUsage()
  n78 = Multiply(A=Multiply(A=Add(A=@n64, B=@n76), B=Power(Base=@n77) {constExponent=11}), B=NamedRerouteUsage())
  n79 = NamedRerouteUsage()
  n80 = NamedRerouteUsage()
  NamedRerouteDeclaration(None=DepthFade(Opacity=LinearInterpolate(A=@n78, B=@n79, Alpha=@n80), FadeDistance=ScalarParameter() {defaultValue=5, parameterName=Depth Fade Distance})) {name=Opacity, nodeColor=<rgba 0.6172 0 1 1>}
  NamedRerouteDeclaration(None=@n17.Contact Point Size) {name=Contact Point Size, nodeColor=<rgba 0.5938 0 1 1>}
  n81 = MaterialFunctionCall(Gradient=@n9, Noise Texture=@n10, Noise Frequencies=@n11, Noise Speeds=@n12.RGBA, Noise Weights=@n13.RGBA, Time=Add(A=@n14) {constB=68.215}) {materialFunction=&MF_Contrail_CreamSkull_Dove_LightningNoise}
  NamedRerouteDeclaration(None=VertexInterpolator(VS=@n81.Result).PS) {name=Lightning Noise B, nodeColor=<rgba 1 0.5391 0 1>}
  n82 = Abs(None=ConstantBiasScale(None=NamedRerouteUsage()) {bias=-0.5, scale=2})
  n83 = Multiply(A=NamedRerouteUsage()) {constB=4}
  n84 = Divide(A=Subtract(A=OneMinus(None=@n82), B=@n83) {constB=0.15}, B=OneMinus(None=@n83))
  n85 = Multiply(
   A=MaterialFunctionCall!(In=Saturate(None=@n84), Power=ScalarParameter() {defaultValue=2, parameterName=WPO Noise Scurve Exponent}) {materialFunction=/Game/Effects/Fort_Effects/MaterialFunctions/MF_SCurve_Float}.Result
   B=NamedRerouteUsage()
  )
  n86 = NamedRerouteUsage()
  NamedRerouteDeclaration(None=Multiply(A=Multiply(A=@n85, B=@n86), B=ScalarParameter() {defaultValue=1, parameterName=WPO Noise})) {name=WPO Noise, nodeColor=<rgba 0.1484 1 0 1>}
  n87 = TextureCoordinate()
  n88 = ComponentMask(None=@n87) {r=true}
  n89 = ComponentMask(None=@n87) {g=true}
  n90 = MaterialFunctionCall(
   Gradient=@n89
   Noise Texture=TextureObject() {samplerType=SAMPLERTYPE_LinearGrayscale, texture=&T_BlueNoise}
  ) {materialFunction=&MF_Contrail_CreamSkull_Dove_LightningNoise}
  n91 = ConstantBiasScale(None=@n89) {bias=-0.5, scale=2}
  n92 = Divide(A=OneMinus(None=Abs(None=@n91)), B=ScalarParameter() {defaultValue=0.1, parameterName=UV Jitter Seam Reduction Size})
  n93 = Multiply(A=@n90.Result, B=Multiply(A=ScalarParameter() {defaultValue=0.15, parameterName=UV Jitter Intensity}, B=Saturate(None=@n92)))
  NamedRerouteDeclaration(None=AppendVector(A=Add(A=@n88, B=@n93), B=@n89)) {name=UV Jittered, nodeColor=<rgba 1 0.1406 0 1>}
  NamedRerouteDeclaration(None=ScalarParameter() {parameterName=Idle Lightning}) {name=Idle Mask, nodeColor=<rgba 0.8984 1 0 1>}
  n94 = NamedRerouteUsage()
  n95 = Saturate(None=@n25)
  n96 = OneMinus(None=Abs(None=ConstantBiasScale(None=@n95) {bias=-0.5, scale=2}))
  n97 = ScalarParameter() {defaultValue=1, parameterName=Idle Lightning Extremities Fade Size}
  n98 = Divide(A=Subtract(A=ComponentMask(None=@n94) {r=true}) {constB=0.5}, B=Saturate(None=Divide(A=@n96, B=@n97)))
  n99 = AppendVector(
   A=Multiply(A=ComponentMask(None=@n94) {g=true}, B=ScalarParameter() {defaultValue=3, parameterName=Idle Lightning UV Tile})
   B=Saturate(None=Add(A=@n98) {constB=0.5})
  )
  n100 = TextureSample(UVs=@n99) {
   samplerType=SAMPLERTYPE_LinearGrayscale
   texture=&T_Contrail_CreamSkull_Dove_Lightning
  }
  n101 = Power(Base=@n100.R)
  n102 = TextureSample(UVs=@n99) {
   samplerType=SAMPLERTYPE_LinearGrayscale
   texture=&T_Contrail_CreamSkull_Dove_Lightning_Blur
  }
  n103 = Multiply(A=@n102.R, B=ScalarParameter() {defaultValue=0.25, parameterName=Idle Lightning Blur Intensity})
  n104 = Multiply(A=@n24, B=Saturate(None=ScalarParameter() {defaultValue=1, parameterName=Idle Lightning Relative Duration}))
  n105 = Subtract(A=Frac(None=@n19), B=Multiply(A=ComponentMask(None=@n20) {b=true, desc=Random Time Offset}, B=OneMinus(None=@n104)))
  n106 = MaterialFunctionCall!(
   Gradient=Frac(None=@n95)
   Erosion Amount=Saturate(None=Divide(A=@n105, B=@n104))
   Softness=Divide(A=ScalarParameter() {defaultValue=0.1, parameterName=Idle Lightning Softness}, B=@n24)
  ) {materialFunction=/BRCosmetics/Athena/Effects/Particles/Emotes/NS_Emote_Spectrum/Functions/MF_Erosion}
  n107 = ScalarParameter() {defaultValue=0.85, parameterName=Idle Lightning Gradient Reflected Center}
  NamedRerouteDeclaration(None=Multiply(
    A=Add(A=@n101, B=@n103)
    B=MaterialFunctionCall!(Gradient=@n106.Increased Range, Center Point=@n107) {materialFunction=/Game/Effects/Fort_Effects/MaterialFunctions/MF_GradientReflected}.Result
   )) {name=Idle Lightning Opacity, nodeColor=<rgba 0 0.9766 1 1>}
  n108 = WorldPosition()
  n109 = NamedRerouteUsage()
  n110 = Normalize(VectorInput=Subtract(A=CameraPositionWS(), B=Add(A=@n108.XYZ, B=@n109)))
  n111 = ScalarParameter() {defaultValue=11, parameterName=Camera Offset}
  NamedRerouteDeclaration(None=Add(A=Multiply(A=@n110, B=@n111), B=NamedRerouteUsage())) {name=WPO, nodeColor=<rgba 1 0 0.5391 1>}
  NamedRerouteDeclaration(None=@n17.Contact Point Fade) {name=Contact Point Fade, nodeColor=<rgba 1 0 0.4922 1>}
  NamedRerouteDeclaration(None=@n17.Contact Point Opacity) {name=Contact Points Brightness, nodeColor=<rgba 0 0.5078 1 1>}
  n112 = Subtract(A=CameraPositionWS(), B=Add(A=@n1.XYZ, B=NamedRerouteUsage()))
  n113 = Constant3Vector() {constant=<rgba 0 1 0 0>}
  n114 = Normalize(VectorInput=CrossProduct(A=Normalize(VectorInput=@n112), B=Transform(None=@n113)))
  n115 = ComponentMask(None=TextureCoordinate()) {r=true}
  n116 = NamedRerouteUsage()
  n117 = Multiply(
   A=ScalarParameter() {defaultValue=33, parameterName=Width}
   B=LinearInterpolate(A=Power(Base=@n116) {constExponent=4}, B=NamedRerouteUsage(), Alpha=NamedRerouteUsage()) {constB=0}
  )
  n118 = NamedRerouteUsage()
  n119 = Multiply(A=Multiply(A=Subtract(A=@n115) {constB=0.5}, B=Multiply(A=@n117, B=@n118)))
  NamedRerouteDeclaration(None=Add(A=Multiply(A=@n114, B=@n119), B=NamedRerouteUsage())) {name=SplineThicken WPO, nodeColor=<rgba 0.2891 0 1 1>}
  NamedRerouteUsage()
  NamedRerouteUsage()
  n120 = MaterialFunctionCall(Gradient=@n9, Noise Texture=@n10, Noise Frequencies=@n11, Noise Speeds=@n12.RGBA, Noise Weights=@n13.RGBA, Time=Add(A=@n14) {constB=95.125}) {materialFunction=&MF_Contrail_CreamSkull_Dove_LightningNoise}
  VertexInterpolator(VS=@n120.Result)
 ]
}
