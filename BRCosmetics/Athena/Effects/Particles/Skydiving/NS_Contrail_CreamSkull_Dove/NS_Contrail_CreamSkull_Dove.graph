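# Skydiving contrail cosmetic that draws a mesh clock oriented to player velocity and facing, reacting to the diving state.
# A smoothed dive value drives Main's hand spin, rim direction and mesh swap, glow, light, floating rim debris and comet spawning.
# CometRibbons and CometSparks trail Main's comets through particle attribute readers.
# LightningBolts spawns on a velocity-aligned ring while diving, whereas Debris's continuous spawning runs only when not diving.
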NiagaraSystem /BRCosmetics/Athena/Effects/Particles/Skydiving/NS_Contrail_CreamSkull_Dove/NS_Contrail_CreamSkull_Dove
dependencies {
 usedDataInterfaces [
  &NiagaraDataInterfaceCurve
  &NiagaraDataInterfaceColorCurve
  &NiagaraDataInterfaceParticleRead
  &NiagaraDataInterfaceSpriteRendererInfo
  &NiagaraDataInterfaceSocketReader
 ]
 usedDynamicInputs [
  &FN_FrontEndVector3Modifier
  /Niagara/DynamicInputs/Math/Clamp_Vector
  &TransformVector
  &MakeCustomVectorFromBool
  &AngleConversion
 ]
 usedModules [
  &self:SystemSpawnScript.SystemScriptSource.SystemScriptGraph.NiagaraNodeAssignment_0.SetVariables_#1
  &SystemState
  &self:SystemSpawnScript.SystemScriptSource.SystemScriptGraph.NiagaraNodeAssignment_1.SetVariables_#2
  &self:SystemSpawnScript.SystemScriptSource.SystemScriptGraph.NiagaraNodeAssignment_2.SetVariables_#3
  &self:SystemSpawnScript.SystemScriptSource.SystemScriptGraph.NiagaraNodeAssignment_5.SetVariables_#4
  &self:SystemSpawnScript.SystemScriptSource.SystemScriptGraph.NiagaraNodeAssignment_3.SetVariables_#5
  &self:SystemSpawnScript.SystemScriptSource.SystemScriptGraph.NiagaraNodeAssignment_4.SetVariables_#6
 ]
}
emitters {
 CometRibbons {
  emitterSpawnScript [/SetVariables(Emitter.RibbonWidth=33) {isSetParameters=true, module=SetVariables_#7}]
  emitterUpdateScript [
   &EmitterState()
   &SpawnParticlesFromOtherEmitter(
    Attribute Reader=DataInterface() {propertyValues {EmitterBinding {bindingMode=Other, emitterName=Main}}}
    Spawn Rate=50
    Use Maximum Age=false
    Use Minimum Age=false
   )
  ]
  particleSpawnScript [
   &InitializeParticle(Color Mode=::NewEnumerator0, UsePositionOffset=false, Write Material Random=true)
   &SampleParticlesFromOtherEmitter(
    Age Sampling:ENiagara_DirectReadSamplingMode=::NewEnumerator0
    Color Sampling=::NewEnumerator0
    Lifetime Application:ENiagara_LifetimeInheritanceOptions=::NewEnumerator1
    Lifetime Sampling:ENiagara_DirectReadSamplingMode=::NewEnumerator0
    Minimum Sampled Particle Age=0.01
    Module.MaxSampledAgeCutoff=false
    Module.MinSampledAgeCutoff=true
    Particle ID Sampling=::NewEnumerator1
    Reject Particle By Sampled Age=true
   )
   /SetVariables(Particles.NormalizedGuideParticleAge:NiagaraFloat=&Divide_Float(A=$Output.SampleParticlesFromOtherEmitter.SampledAge, B=$Output.SampleParticlesFromOtherEmitter.SampledLifetime)) {isSetParameters=true, module=SetVariables_#8}
   /SetVariables(Particles.RibbonWidth=&FloatFromCurve(
     FloatCurve=DataInterface() {propertyValues {Curve {defaultValue=0.5, keys [{tangentMode=RCTM_User, time=0, value=0.5}, {tangentMode=RCTM_User, time=1, value=1}]}}}
     CurveIndex=$Particles.NormalizedGuideParticleAge
     Scale Curve=$Emitter.RibbonWidth
    )) {isSetParameters=true}
  ]
  particleUpdateScript [
   &ParticleState()
   &SolveForcesAndVelocity(Clamp Velocity=false, Limit Acceleration=false)
   &Color(Color=&MakeLinearColorFromVectorAndFloat(
     "Vector (RGB)"=&MakeVectorFromLinearColorRGB(Color=&LerpLinearColors(StartColor=$Particles.Initial.Color, EndColor=<rgba 0.49 0.1 1 1>, LerpFactor=$Particles.NormalizedAge))
     "Float (Alpha)"=&FloatFromCurve(FloatCurve=DataInterface() {propertyValues {Curve {
         defaultValue=1
         keys [{interpMode=RCIM_Cubic, leaveTangent=-2.919, tangentMode=RCTM_Break, time=0, value=1}, {tangentMode=RCTM_User, time=1, value=0}]
        }}})
    ))
  ]
  rendererClasses [&NiagaraRibbonRendererProperties]
  renderers [{rendererClass=&NiagaraRibbonRendererProperties}]
 }
 CometSparks {
  emitterSpawnScript [/SetVariables(Emitter.RibbonWidth=15) {isSetParameters=true, module=SetVariables_#7}]
  emitterUpdateScript [
   &EmitterState()
   &SpawnParticlesFromOtherEmitter(
    Attribute Reader=DataInterface() {propertyValues {EmitterBinding {bindingMode=Other, emitterName=Main}}}
    Spawn Rate=50
    Use Maximum Age=false
    Use Minimum Age=false
   )
  ]
  particleSpawnScript [
   &InitializeParticle(
    Color=<rgba 9 2.7089 1.9759 1>
    Lifetime Max=1.3
    Lifetime Min=0.62
    Lifetime Mode=::NewEnumerator1
    Sprite Size Max=<3 8>
    Sprite Size Min=<1.5 4>
    Sprite Size Mode=::NewEnumerator2
    UsePositionOffset=false
    Write Material Random=true
   )
   &SampleParticlesFromOtherEmitter(
    Age Sampling:ENiagara_DirectReadSamplingMode=::NewEnumerator1
    Applied Velocity Scale=0.5
    Color Sampling=::NewEnumerator0
    Lifetime Sampling:ENiagara_DirectReadSamplingMode=::NewEnumerator1
    Minimum Sampled Particle Age=0.01
    Module.MaxSampledAgeCutoff=false
    Module.MinSampledAgeCutoff=true
    Reject Particle By Sampled Age=true
    Velocity Sampling=::NewEnumerator0
   )
   /SetVariables(Particles.NormalizedGuideParticleAge:NiagaraFloat=&Divide_Float(A=$Output.SampleParticlesFromOtherEmitter.SampledAge, B=$Output.SampleParticlesFromOtherEmitter.SampledLifetime)) {isSetParameters=true, module=SetVariables_#8}
   &ShapeLocation(MaskSpawnGroup=false, Override Seed=false, Sphere Radius=2, UseOwnerScale=true)
   &AddVelocity(
    Module.Use Falloff Distance=false
    Override Seed=false
    Velocity Mode=::NewEnumerator1
    Velocity Origin=$Output.SampleParticlesFromOtherEmitter.SampledPosition
   )
  ]
  particleUpdateScript [
   &ParticleState()
   &Color(Color=&MakeLinearColorFromVectorAndFloat(
     "Vector (RGB)"=&MakeVectorFromLinearColorRGB(Color=&LerpLinearColors(StartColor=$Particles.Initial.Color, EndColor=<rgba 0.49 0.1 1 1>, LerpFactor=$Particles.NormalizedAge))
     "Float (Alpha)"=&FloatFromCurve(FloatCurve=DataInterface() {propertyValues {Curve {
         defaultValue=1
         keys [{interpMode=RCIM_Cubic, leaveTangent=-2.919, tangentMode=RCTM_Break, time=0, value=1}, {tangentMode=RCTM_User, time=1, value=0}]
        }}})
    ))
   &Drag(Use Linear Drag=true, Use Rotational Drag=true)
   &SolveForcesAndVelocity(Clamp Velocity=false, Limit Acceleration=false)
   /SetVariables(
    Particles.SparkGlowColor:LinearColor=&Multiply_LinearColorByFloat(Linear Color=$Particles.Color, Float=0.3)
    Particles.SparkGlowSize:Vector2f=&Vector2DFromFloat(Value=22)
   ) {isSetParameters=true}
  ]
  rendererClasses [&NiagaraSpriteRendererProperties]
  renderers [
   {rendererClass=&NiagaraSpriteRendererProperties}
   {rendererClass=&NiagaraSpriteRendererProperties, rendererIndex=1}
  ]
 }
 Debris {
  emitterUpdateScript [
   &EmitterState(Life Cycle Mode=::NewEnumerator1, Loop Behavior=::NewEnumerator1, Loop Duration Mode=::NewEnumerator1, UseLoopDelay=false)
   &SpawnRate(
    Spawn Probability=&MakeCustomFloatFromBool(Bool=$System.IsDiving, True Float=0, False Float=1)
    SpawnGroup=1
    SpawnRate=5
    Use Spawn Probability=true
   ) {module=SpawnRate001}
   &SpawnRate(
    Spawn Probability=&MakeCustomFloatFromBool(Bool=$System.IsDiving, True Float=0, False Float=1)
    SpawnRate=11
    Use Spawn Probability=true
   )
   &SpawnBurst_Instantaneous(Spawn Count=5, Use Spawn Probability=false, UseLoopCountLimit=false)
   /SetVariables(Emitter.MeshReferenceAxisVector:Vector3f=<0 0 -1>) {isSetParameters=true, module=SetVariables_#9}
   /SetVariables(
    Emitter.MeshRotation:Quat4f=&MakeQuaternion(
     Quaternion From=3
     Axis=<1 0 0>
     Angle=0
     From=$Emitter.MeshReferenceAxisVector
     To=$System.InterpolatedLocalVelocity
     Angle Type=::NewEnumerator0
     Coordinate Space=::Local
     Normalize Vectors=true
    )
    Emitter.Position:NiagaraPosition=&AddVectorToPosition(Position=$Engine.Emitter.SimulationPosition, Vector=$System.InterpolatedPositionOffset)
   ) {isSetParameters=true, module=SetVariables_#10}
  ]
  particleSpawnScript [
   &InitializeParticle(
    Color=&MaskLinearColorBySpawnGroup(Masked Linear Color=<rgba 0.1 0.064 0.04 1>, Passthrough Linear Color=<rgba 0.15 0.1365 0.12 1>)
    Sprite Size Mode=::NewEnumerator1
    UsePositionOffset=false
    Write Material Random=true
   )
   &ShapeLocation(
    Additional Quaternion Rotation=false
    MaskSpawnGroup=false
    Offset=<0 0 -111>
    Offset Mode=::NewEnumerator0
    Override Seed=false
    Ring Radius=99
    Rotation Mode=::NewEnumerator3
    Rotation Quaternion=$Emitter.MeshRotation
    Shape Origin=$Emitter.Position
    Shape Primitive=::NewEnumerator4
    UseOwnerScale=true
   )
   &AddVelocity(Velocity=<0 0 333>)
   /SetVariables(Particles.VisibilityTag=&SpawnGroup()) {isSetParameters=true}
  ]
  particleUpdateScript [
   &ParticleState()
   &ScaleSpriteSize(Uniform Curve Sprite Scale=DataInterface() {propertyValues {Curve {
       defaultValue=0
       keys [
        {arriveTangent=4.1785, interpMode=RCIM_Cubic, leaveTangent=4.1785, tangentMode=RCTM_Break, time=0, value=0}
        {arriveTangent=0.0224, interpMode=RCIM_Cubic, leaveTangent=0.0224, tangentMode=RCTM_Break, time=0.75, value=1}
        {arriveTangent=-8.0219, interpMode=RCIM_Cubic, tangentMode=RCTM_Break, time=1, value=0}
       ]
      }}})
   &SpriteRotationRate(Rotation Rate=&RandomRangeFloat(Minimum=-666, Maximum=666))
   &CurlNoiseForce(Bias Noise=false, Coordinate Space:ENiagaraCoordinateSpace=::Local, Noise Frequency=222, Noise Strength=1111, Pan Noise=false)
   &SolveForcesAndVelocity(Clamp Velocity=false, Limit Acceleration=false)
   &SubUVAnimation(
    Random Start Frame=true
    "Sprite Renderer "=DataInterface() {propertyValues {SpriteRenderer=&self:Debris_0.NiagaraSpriteRendererProperties_1}}
    SubUV Lookup Index Scale=2
    UseEndFrame=false
    UseStartFrame=false
   )
   &SubUVAnimation(
    Random Start Frame=true
    "Sprite Renderer "=DataInterface() {propertyValues {SpriteRenderer=&self:Debris_0.NiagaraSpriteRendererProperties_0}}
    SubUV Lookup Index Scale=2
    UseEndFrame=false
    UseStartFrame=false
   ) {module=SubUVAnimation001}
  ]
  rendererClasses [&NiagaraSpriteRendererProperties]
  renderers [
   {rendererClass=&NiagaraSpriteRendererProperties}
   {rendererClass=&NiagaraSpriteRendererProperties, rendererIndex=1}
  ]
 }
 LightningBolts {
  emitterUpdateScript [
   &EmitterState()
   &SpawnRate(
    Spawn Probability=&MakeCustomFloatFromBool(Bool=$System.IsDiving, True Float=1, False Float=0)
    SpawnRate=11
    Use Spawn Probability=true
   )
   /SetVariables(Emitter.MeshReferenceAxisVector:Vector3f=<0 0 -1>) {isSetParameters=true, module=SetVariables_#9}
   /SetVariables(
    Emitter.MeshRotation:Quat4f=&MakeQuaternion(
     Quaternion From=3
     Axis=<1 0 0>
     Angle=0
     From=$Emitter.MeshReferenceAxisVector
     To=$System.InterpolatedLocalVelocity
     Angle Type=::NewEnumerator0
     Coordinate Space=::Local
     Normalize Vectors=true
    )
    Emitter.Position:NiagaraPosition=&AddVectorToPosition(Position=$Engine.Emitter.SimulationPosition, Vector=$System.InterpolatedPositionOffset)
   ) {isSetParameters=true, module=SetVariables_#10}
  ]
  particleSpawnScript [
   &InitializeParticle(
    Color Mode=::NewEnumerator0
    Lifetime Max=0.6
    Lifetime Min=0.45
    Lifetime Mode=::NewEnumerator1
    Sprite Size=<44 222>
    Sprite Size Mode=::NewEnumerator4
    UsePositionOffset=false
    Write Material Random=true
   )
   &ShapeLocation(
    Additional Quaternion Rotation=false
    MaskSpawnGroup=false
    Override Seed=false
    Ring Radius=66
    Rotation Mode=::NewEnumerator3
    Rotation Quaternion=$Emitter.MeshRotation
    Shape Origin=$Emitter.Position
    Shape Primitive=::NewEnumerator4
    UseOwnerScale=true
   )
  ]
  particleUpdateScript [
   &ParticleState()
   &SpriteFacingAndAlignment(
    Alignment Coordinate Space=::Local
    Sprite Alignment=&Lerp_Vector(
     A=&NormalizeVector(
      Vector To Normalize=$Particles.Initial.ShapeLocation.ShapeVector
      Safe Normalize=true
      Safe Fallback Threshold=0
      Safe Fallback Vector=<0 0 0.001>
     )
     B=$Particles.ShapeLocation.ShapeNormal
     Alpha=0.8
    )
    WriteAlignment=true
    WriteFacing=false
   )
  ]
  rendererClasses [&NiagaraSpriteRendererProperties]
  renderers [{rendererClass=&NiagaraSpriteRendererProperties}]
 }
 Main {
  emitterSpawnScript [
   /SetVariables(
    Emitter.ClockLargeHandAngleLerpRate=1
    Emitter.ClockLargeHandRotationAngle=0.07
    Emitter.ClockLargeHandRotationSpeed=-3
    Emitter.ClockRimRotationSpeed=0.06
    Emitter.ClockSmallHandAngleLerpRate=1
    Emitter.ClockSmallHandRotationAngle=-0.166
    Emitter.ClockSmallHandRotationSpeed=-0.1
   ) {isSetParameters=true}
   /SetVariables(Emitter.MeshScaleMultiplier=0.666) {isSetParameters=true}
   /SetVariables(
    Emitter.ClockRimMeshTransitionPoint=0.2
    Emitter.MeshScale:Vector3f=&VectorFromFloat(Value=$Emitter.MeshScaleMultiplier)
   ) {isSetParameters=true}
  ]
  emitterUpdateScript [
   &EmitterState(
    Inactive Response=::NewEnumerator1
    Life Cycle Mode=::NewEnumerator1
    Loop Behavior=::NewEnumerator1
    Loop Duration Mode=::NewEnumerator1
    UseLoopDelay=false
   )
   &SpawnBurst_Instantaneous(Spawn Count=28, Spawn Group=1, Use Spawn Probability=false, UseLoopCountLimit=false)
   &InterpolateOverTime(
    Rate Of Change=2
    Target Float Value:NiagaraFloat=&MakeCustomFloatFromBool(Bool=$System.IsDiving, True Float=1, False Float=0)
   )
   &SpawnRate(
    SpawnRate=&Multiply_Float(A=$Emitter.InterpolateOverTime.InterpolateOverTime_Float.Moving Average, B=1.5)
    Use Spawn Probability=false
   )
   /SetVariables(Emitter.MeshReferenceAxisVector:Vector3f=<0 0 -1>, Emitter.PlayerReferenceAxisVector:Vector3f=<0 -1 0>) {isSetParameters=true}
   /SetVariables(
    Emitter.MeshRotation:Quat4f=&MakeQuaternion(
     Quaternion From=3
     Axis=<1 0 0>
     Angle=0
     From=$Emitter.MeshReferenceAxisVector
     To=&NormalizeVector(Vector To Normalize=$System.InterpolatedLocalVelocity, Safe Normalize=true, Safe Fallback Threshold=0, Safe Fallback Vector=<0 0 0.001>)
     Angle Type=::NewEnumerator0
     Coordinate Space=::Local
     Normalize Vectors=false
    )
    Emitter.PlayerRotation:Quat4f=&MakeQuaternion(
     Quaternion From=3
     Axis=$System.InterpolatedLocalPlayerVector
     Angle=0
     From=$Emitter.PlayerReferenceAxisVector
     To=$System.InterpolatedLocalPlayerVector
     Angle Type=::NewEnumerator0
     Coordinate Space=::Local
     Normalize Vectors=true
    )
    Emitter.Position:NiagaraPosition=&AddVectorToPosition(Position=$Engine.Emitter.SimulationPosition, Vector=$System.InterpolatedPositionOffset)
   ) {isSetParameters=true}
   /SetVariables(
    Emitter.ClockLargeHandRotationAngle:NiagaraFloat=&MakeCustomFloatFromBool(
     Bool=$System.IsDiving
     True Float=&Add_Float(
      A=$Emitter.ClockLargeHandRotationAngle
      B=&Multiply_Float(
       A=$Engine.DeltaTime
       B=&Multiply_Float(A=$Emitter.InterpolateOverTime.InterpolateOverTime_Float.Moving Average, B=$Emitter.ClockLargeHandRotationSpeed)
      )
     )
     False Float=&Lerp_Float(
      A=&Frac_Float(Float=$Emitter.ClockLargeHandRotationAngle)
      B=$Emitter.Initial.ClockLargeHandRotationAngle
      Alpha=&Multiply_Float(A=$Engine.DeltaTime, B=$Emitter.ClockLargeHandAngleLerpRate)
     )
    )
    Emitter.ClockRimRotationAngle:NiagaraFloat=&Add_Float(
     A=$Emitter.ClockRimRotationAngle
     B=&Multiply_Float(
      A=$Engine.DeltaTime
      B=&Lerp_Float(
       A=&Multiply_Float(A=$Emitter.ClockRimRotationSpeed, B=-1)
       B=$Emitter.ClockRimRotationSpeed
       Alpha=$Emitter.InterpolateOverTime.InterpolateOverTime_Float.Moving Average
      )
     )
    )
    Emitter.ClockSmallHandRotationAngle:NiagaraFloat=&MakeCustomFloatFromBool(
     Bool=$System.IsDiving
     True Float=&Add_Float(
      A=$Emitter.ClockSmallHandRotationAngle
      B=&Multiply_Float(
       A=$Engine.DeltaTime
       B=&Multiply_Float(A=$Emitter.InterpolateOverTime.InterpolateOverTime_Float.Moving Average, B=$Emitter.ClockSmallHandRotationSpeed)
      )
     )
     False Float=&Lerp_Float(
      A=&Frac_Float(Float=$Emitter.ClockSmallHandRotationAngle)
      B=$Emitter.Initial.ClockSmallHandRotationAngle
      Alpha=&Multiply_Float(A=$Engine.DeltaTime, B=$Emitter.ClockSmallHandAngleLerpRate)
     )
    )
   ) {isSetParameters=true}
   /SetVariables(
    Emitter.ClockLargeHandLocalRotation:Quat4f=&MakeQuaternion(
     Quaternion From=0
     Axis=<0 0 1>
     Angle=$Emitter.ClockLargeHandRotationAngle
     From=?
     To=?
     Angle Type=::NewEnumerator1
     Coordinate Space=::Local
     Normalize Vectors=false
    )
    Emitter.ClockRimLocalRotation:Quat4f=&MakeQuaternion(
     Quaternion From=0
     Axis=<0 0 1>
     Angle=$Emitter.ClockRimRotationAngle
     From=?
     To=?
     Angle Type=::NewEnumerator1
     Coordinate Space=::Local
     Normalize Vectors=false
    )
    Emitter.ClockSmallHandLocalRotation:Quat4f=&MakeQuaternion(
     Quaternion From=0
     Axis=<0 0 1>
     Angle=$Emitter.ClockSmallHandRotationAngle
     From=?
     To=?
     Angle Type=::NewEnumerator1
     Coordinate Space=::Local
     Normalize Vectors=false
    )
   ) {isSetParameters=true}
   /SetVariables(
    Emitter.ClockRimMeshIndex:NiagaraInt32=&MakeCustomIntFromBool(
     Bool=&SetBoolByFloatComparison(A=$Emitter.InterpolateOverTime.InterpolateOverTime_Float.Moving Average, B=0.2)
     True INT=1
     False INT=0
    )
    Emitter.IdleLightning:NiagaraFloat=&MakeCustomFloatFromBool(Bool=$System.IsDiving, True Float=0, False Float=1)
    Emitter.MaterialEmissive:NiagaraFloat=$Emitter.InterpolateOverTime.InterpolateOverTime_Float.Moving Average
   ) {isSetParameters=true}
   /SetVariables(Emitter.ClockFaceRotation:Quat4f=&MultiplyQuaternion(Quaternion A=$Emitter.MeshRotation, Quaternion B=$Emitter.PlayerRotation)) {isSetParameters=true}
   /SetVariables(
    Emitter.ClockLargeHandRotation:Quat4f=&MultiplyQuaternion(Quaternion A=$Emitter.ClockFaceRotation, Quaternion B=$Emitter.ClockLargeHandLocalRotation)
    Emitter.ClockRimRotation:Quat4f=&MultiplyQuaternion(Quaternion A=$Emitter.ClockFaceRotation, Quaternion B=$Emitter.ClockRimLocalRotation)
    Emitter.ClockSmallHandRotation:Quat4f=&MultiplyQuaternion(Quaternion A=$Emitter.ClockFaceRotation, Quaternion B=$Emitter.ClockSmallHandLocalRotation)
   ) {isSetParameters=true}
   /SetVariables(
    Emitter.LightColor:LinearColor=&Multiply_LinearColorByFloat(Linear Color=<rgba 13.2 23.1 33 1>, Float=$Emitter.MaterialEmissive)
    Emitter.LightEnabled:NiagaraBool=&SetBoolByFloatComparison(A=$Emitter.MaterialEmissive, B=0.01)
   ) {isSetParameters=true}
  ]
  particleSpawnScript [
   &InitializeParticle(
    Color=<rgba 3 0.5953 1.723 1>
    Lifetime Max=1.2
    Lifetime Min=0.4
    Lifetime Mode=::NewEnumerator1
    Sprite Size Mode=::NewEnumerator3
    Uniform Sprite Size=55
    UsePositionOffset=false
    Write Material Random=true
   )
   &ShapeLocation(
    Additional Quaternion Rotation=false
    MaskSpawnGroup=false
    Override Seed=false
    Ring Radius=66
    Rotation Mode=::NewEnumerator3
    Rotation Quaternion=$Emitter.MeshRotation
    Shape Origin=$Emitter.Position
    Shape Primitive=::NewEnumerator4
    UseOwnerScale=true
   )
   &AddVelocity(
    Velocity=&Add_Vector(
     A=&Multiply_VectorByFloat(
      Vector=&NormalizeVector(Vector To Normalize=$System.InterpolatedLocalVelocity, Safe Normalize=true, Safe Fallback Threshold=0, Safe Fallback Vector=<0 0 0.001>)
      Float=-444
     )
     B=&Multiply_VectorByFloat(
      Vector=&NormalizeVector(Vector To Normalize=$Particles.ShapeLocation.ShapeVector, Safe Normalize=true, Safe Fallback Threshold=0, Safe Fallback Vector=<0 0 0.001>)
      Float=222
     )
    )
    Velocity Speed Scale=1.5
   )
   /SetVariables(
    Particles.DebrisCameraOffset=22
    Particles.DebrisColor:LinearColor=<rgba 0 0 0 1>
    Particles.DebrisSize:Vector2f=&Multiply_Vector2D_ByFloat(
     Float=&Lerp_Float(
      A=2
      B=9
      Alpha=&FloatFromCurve(
       FloatCurve=DataInterface() {propertyValues {Curve {keys [{interpMode=RCIM_Cubic, tangentMode=RCTM_Break, time=0, value=0}, {arriveTangent=2.5678, tangentMode=RCTM_User, time=1, value=1}]}}}
       CurveIndex=&RandomRangeFloat(Minimum=0, Maximum=1)
      )
     )
     Vector2D=<1.1 0.9>
    )
    Particles.DebrisSubImageIndex:NiagaraFloat=&RandomRangeFloat(Minimum=0, Maximum=9)
   ) {isSetParameters=true}
  ]
  particleUpdateScript [
   &KillParticles(Kill Particles=/Niagara/DynamicInputs/Helpers/MaskBoolBySpawnGroup(
     Masked Bool:NiagaraBool=&SetBoolByFloatComparison(A=$Particles.NormalizedAge, B=1)
     Passthrough Bool:NiagaraBool=false
     Spawn Group Mask:NiagaraInt32=0
    ))
   &ParticleState(Kill Particles When Lifetime Has Elapsed=false)
   &ScaleSpriteSize(Uniform Curve Sprite Scale=DataInterface() {propertyValues {Curve {defaultValue=0, keys [{tangentMode=RCTM_User, time=0, value=0}, {tangentMode=RCTM_User, time=0.2, value=1}, <at 0.75 1>, <at 1 0>]}}})
   &CurlNoiseForce(Bias Noise=false, Noise Frequency=22, Noise Strength=2500, Pan Noise=false)
   &SolveForcesAndVelocity(Clamp Velocity=false, Limit Acceleration=false)
   /SetVariables(
    Particles.CometGlowColor:LinearColor=&Multiply_LinearColorByFloat(Linear Color=$Particles.Color, Float=0.2)
    Particles.CometGlowSize:Vector2f=&Multiply_Vector2D_ByFloat(Float=3, Vector2D=$Particles.SpriteSize)
   ) {isSetParameters=true}
   /SetVariables(
    Particles.Age:NiagaraFloat=&MaskFloatBySpawnGroup(Masked Float=$Particles.Age, Passthrough Float=0)
    Particles.DebrisColor:LinearColor=&ColorFromCurve(
     ColorCurve=DataInterface() {propertyValues {
       AlphaCurve {keys [<at 0.2 0 RCIM_Cubic>,<at 0.25 1>]}
       BlueCurve {keys [<at 0 1 RCIM_Cubic>,<at 0.22 1>,<at 0.5 0.2>,<at 0.8 0 RCIM_Cubic>]}
       GreenCurve {keys [<at 0 1 RCIM_Cubic>,<at 0.22 1>,<at 0.5 0.2>,<at 0.8 0 RCIM_Cubic>]}
       RedCurve {keys [<at 0 1 RCIM_Cubic>,<at 0.22 1>,<at 0.5 0.2>,<at 0.8 0 RCIM_Cubic>]}
      }}
     CurveIndex=$Emitter.MaterialEmissive
    )
    Particles.DebrisPosition:NiagaraPosition=&AddVectorToPosition(
     Position=&AddVectorToPosition(
      Position=$Emitter.Position
      Vector=&RotateVectorByQuaternion(
       Vector To Rotate=/Niagara/DynamicInputs/Vectors/RotateVector(
        Vector To Rotate:Vector3f=&Lerp_Vector(
         A=<0 60 0>
         B=<0 79 0>
         Alpha=&FloatFromCurve(FloatCurve=DataInterface() {propertyValues {Curve {keys [<at 0.2 0>,<at 1 1>]}}}, CurveIndex=$Emitter.MaterialEmissive)
        )
        Yaw:NiagaraFloat=&SelectFloatFromArray(Array Sampling Mode=::NewEnumerator0, Direct Array Index=?, Float Selection Array=$User.DebrisYawAngles, Random Seed=0)
        Pitch:NiagaraFloat=0
        Roll:NiagaraFloat=0
        Rotation Rate:NiagaraFloat=1
       )
       Quaternion=$Emitter.ClockRimRotation
      )
     )
     Vector=&RotateVector(
      Vector To Rotate:Vector3f=&RandomVector(Vector Scale=&Waveform(
        Global Phase=&Add_Float(A=$Emitter.Age, B=&RandomRangeFloat(Minimum=0, Maximum=99))
        Global Amplitude Scale=&FloatFromCurve(
         FloatCurve=DataInterface() {propertyValues {Curve {keys [<at 0.2 0>,<at 1 1>]}}}
         CurveIndex=$Emitter.MaterialEmissive
         Scale Curve=&RandomRangeFloat(Minimum=11, Maximum=22)
        )
        "Waveform [1]"=::NewEnumerator12
        "[1] Phase Scale"=0.2
        "[1] Frequency"=1
        "[1] Amplitude Min/Max"=<-1 1>
       ))
      Yaw:NiagaraFloat=26
      Pitch:NiagaraFloat=36
      Roll:NiagaraFloat=43
      Rotation Rate:NiagaraFloat=&Add_Float(
       A=&Multiply_Float(A=$Emitter.Age, B=&RandomRangeFloat(Minimum=0.62, Maximum=1.62))
       B=&RandomRangeFloat(Minimum=0, Maximum=22)
      )
     )
    )
    Particles.DebrisRotation:NiagaraFloat=&Multiply_Float(A=$Emitter.Age, B=&RandomRangeFloat(Minimum=-222, Maximum=222))
    Particles.DebrisSize:Vector2f=&Multiply_Vector2D_ByFloat(
     Float=&FloatFromCurve(
      FloatCurve=DataInterface() {propertyValues {Curve {keys [{interpMode=RCIM_Cubic, tangentMode=RCTM_Break, time=0.2, value=0}, <at 0.75 1>]}}}
      CurveIndex=$Emitter.MaterialEmissive
     )
     Vector2D=$Particles.Initial.DebrisSize
    )
   ) {isSetParameters=true}
   /SetVariables(
    Particles.DebrisGlowColor:LinearColor=<rgba 0.0016 0.0208 0.04 1>
    Particles.DebrisGlowSize:Vector2f=&Vector2DFromFloat(Value=&Multiply_Float(A=2.5, B=/Niagara/DynamicInputs/TypeConversions/Vector2DLength(Vector:Vector2f=$Particles.DebrisSize)))
   ) {isSetParameters=true}
  ]
  rendererClasses [
   &NiagaraMeshRendererProperties
   &NiagaraSpriteRendererProperties
   &NiagaraLightRendererProperties
  ]
  renderers [
   {rendererClass=&NiagaraMeshRendererProperties, rendererIndex=0}
   {rendererClass=&NiagaraMeshRendererProperties, rendererIndex=1}
   {rendererClass=&NiagaraMeshRendererProperties, rendererIndex=2}
   {rendererClass=&NiagaraMeshRendererProperties, rendererIndex=3}
   {rendererClass=&NiagaraMeshRendererProperties, rendererIndex=4}
   {rendererClass=&NiagaraMeshRendererProperties, rendererIndex=5}
   {rendererClass=&NiagaraMeshRendererProperties, rendererIndex=6}
   {rendererClass=&NiagaraMeshRendererProperties, rendererIndex=7}
   {rendererClass=&NiagaraMeshRendererProperties, rendererIndex=8}
   {rendererClass=&NiagaraSpriteRendererProperties, rendererIndex=9}
   {rendererClass=&NiagaraSpriteRendererProperties, rendererIndex=10}
   {rendererClass=&NiagaraMeshRendererProperties, rendererIndex=11}
   {rendererClass=&NiagaraSpriteRendererProperties, rendererIndex=12}
   {rendererClass=&NiagaraSpriteRendererProperties, rendererIndex=13}
   {rendererClass=&NiagaraLightRendererProperties, rendererIndex=14}
  ]
 }
}
summary {
 systemName=NS_Contrail_CreamSkull_Dove
 emitters [
  {
   emitterName=Main
   rendererClasses [
    &NiagaraMeshRendererProperties
    &NiagaraSpriteRendererProperties
    &NiagaraLightRendererProperties
   ]
  }
  {emitterName=CometRibbons, rendererClasses [&NiagaraRibbonRendererProperties]}
  {emitterName=CometSparks, rendererClasses [&NiagaraSpriteRendererProperties]}
  {emitterName=LightningBolts, rendererClasses [&NiagaraSpriteRendererProperties]}
  {emitterName=Debris, rendererClasses [&NiagaraSpriteRendererProperties]}
 ]
 userVariables [
  User.bIsFrontEnd {type=NiagaraBool}
  User.bIsFrontEndPreview {type=NiagaraBool}
  User.DebrisYawAngles {
   type=NiagaraDataInterfaceArrayFloat
   defaultValue {
    struct=NiagaraExt_VariableValue_DataInterface
    value {
     dataInterface=&self:NiagaraDataInterfaceArrayFloat_0
     dataInterfaceClass=&NiagaraDataInterfaceArrayFloat
    }
   }
  }
  User.ParaglideLeanAlpha {type=NiagaraFloat, defaultValue=0, description=""}
 ]
}
